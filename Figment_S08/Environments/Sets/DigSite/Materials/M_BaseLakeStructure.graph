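# Base metal material for DigSite lake structures, blending four texture layers with tints and optional high-quality parallax UVs.
# It adds a view-space chromatic sheen, distance-blended detail noise and normals, and flickering masked emissive, then applies gradient, vertex-paint and up-vector functions.
# A hit-effect mask flattens normals and overrides roughness and color.
# The final World Position Offset comes from the DanceClub switch (zero when off), replacing the hit offset.
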Material /Figment_S08/Environments/Sets/DigSite/Materials/M_BaseLakeStructure
graph {
 properties {
  bUseMaterialAttributes=true
  ditheredLODTransition=true
  physMaterial=/Game/FortressPhysicalMaterials/Metal
  pixelDepthOffsetMode=PDOM_Legacy
  usedShadingModels=Default Lit
 }
 nodes [
  n1 = ParticleRandom()
  n2 = DynamicParameter() {paramNames [FlickerFade,FlickerSpeed,Param3,Param4]}
  n3 = ScalarParameter() {defaultValue=0.2, group=Emissive, parameterName=Emissive Flicker Speed}
  n4 = Multiply(
   A=Time()
   B=StaticSwitchParameter(True=LinearInterpolate(Alpha=@n1) {constA=1.2, constB=0.8}, False=Multiply(A=@n2.FlickerSpeed, B=@n3)) {parameterName=GPU Particles}
  )
  n5 = Sine(None=@n4) {period=0.075}
  n6 = Sine(None=@n4) {period=0.155}
  n7 = Sine(None=@n4) {period=0.2}
  Multiply(A=@n5, B=Multiply(A=Multiply(A=@n6, B=@n7), B=Sine(None=@n4) {period=0.43}))
  n8 = Transform(None=PixelNormalWS()) {transformSourceType=TRANSFORMSOURCE_World, transformType=TRANSFORM_View}
  n9 = Multiply(
   A=Fresnel(ExponentIn=ScalarParameter() {defaultValue=1.0389, group=Chromatic, parameterName=CromaticExponent}) {baseReflectFraction=0}
   B=ScalarParameter() {defaultValue=1.25, group=Chromatic, parameterName=ChromaticFresnelX}
  )
  n10 = ScalarParameter() {defaultValue=-0.8, group=Chromatic, parameterName=ChromaticFresnelAdd}
  n11 = ScalarParameter() {defaultValue=0.1, group=Chromatic, parameterName=ChromaticStaticAmount}
  n12 = TextureSample(UVs=Multiply(A=TextureCoordinate(), B=ScalarParameter() {defaultValue=2, group=Chromatic, parameterName=ChromaticStaticSize})) {texture=/Game/Effects/Fort_Effects/Textures/Tile/T_Inky_Smoke_Tile_Chromatic}
  n13 = AppendVector(
   A=Multiply(A=ComponentMask(None=@n8) {r=true}, B=ScalarParameter() {defaultValue=0.4, group=Chromatic, parameterName=CromaticViewSpaceSize})
   B=Add(A=Add(A=@n9, B=@n10), B=Multiply(A=@n11, B=@n12.R))
  )
  n14 = MaterialFunctionCall!(Blend Range=ScalarParameter() {defaultValue=1000, parameterName=DetailFresRange}, Start Offset=Constant()) {materialFunction=/Engine/Functions/Engine_MaterialFunctions03/Particles/Distance_Blend}
  n15 = Multiply(
   A=Multiply(
    A=TextureSample(UVs=@n13) {texture=&T_Inky_Smoke_Tile_Chromatic}.RGB
    B=ScalarParameter() {defaultValue=0.5, group=Chromatic, parameterName=CromaticMasterAmount}
   )
   B=@n14.Result
  )
  n16 = MaterialFunctionCall() {materialFunction=/Game/Athena/WinterArt/Materials/MF_UVsSet}
  n17 = TextureObjectParameter() {group=Material_1, parameterName=Normals, samplerType=SAMPLERTYPE_Normal, sortPriority=2, texture=/Engine/EngineMaterials/FlatNormal}
  n18 = ScalarParameter() {defaultValue=1, group=Material_1, parameterName=Mat1 UV Scale, sortPriority=3}
  n19 = Multiply(A=TextureCoordinate(), B=Clamp(None=Ceil(None=@n18)) {maxDefault=1000, minDefault=1})
  n20 = TextureSample!(UVs=@n16.POM UVs 4 Layers, Tex=@n17, "DDX(UVs)"=DDX(Value=@n19), "DDY(UVs)"=DDY(Value=@n19), Apply View MipBias) {mipValueMode=TMVM_Derivative, samplerType=SAMPLERTYPE_Normal}
  n21 = TextureSample!(UVs=@n16.UVs Layer 1, Tex=@n17, Apply View MipBias) {samplerType=SAMPLERTYPE_Normal}
  n22 = QualitySwitch(
   Default=@n16.UVs Layer 1
   High=StaticSwitchParameter(True=@n16.POM UVs 4 Layers, False=@n16.POM UVs 4 Layers) {group=Material_1, parameterName=POM_True, sortPriority=11}
  )
  n23 = TextureObjectParameter() {
   group=Material_1
   parameterName=SpecularMasks
   samplerType=SAMPLERTYPE_Masks
   sortPriority=1
   texture=/Game/Packages/Fortnite_UniqueMaterials_1/Textures/T_Fortnite_Default_S
  }
  n24 = TextureSample(UVs=@n22, Tex=@n23) {samplerType=SAMPLERTYPE_Masks, texture=/Game/Environments/Sets/IceCastle/Textures/T_IceCastle_Wall1_D}
  n25 = MaterialFunctionCall() {materialFunction=&MF_UVsSet}
  n26 = QualitySwitch(
   Default=@n25.UVs Layer 3
   High=StaticSwitchParameter(True=@n25.POM UVs 4 Layers, False=@n25.POM UVs 4 Layers) {group=Material_3, parameterName=L3_POM_True, sortPriority=11}
  )
  n27 = TextureSampleParameter2D(UVs=@n26) {
   desc=Use for trims and accents
   group=Material_3
   parameterName=SpecularMasks_3
   samplerSource=SSM_Wrap_WorldGroupSettings
   samplerType=SAMPLERTYPE_Masks
   sortPriority=2
   texture=&T_Fortnite_Default_S
  }
  n28 = TextureObjectParameter() {
   group=Material_1
   parameterName=DM
   samplerType=SAMPLERTYPE_Masks
   sortPriority=1
   texture=&T_Fortnite_Default_S
  }
  n29 = TextureSample(UVs=@n22, Tex=@n28) {samplerType=SAMPLERTYPE_Masks}
  n30 = VectorParameter() {group=Material_1, parameterName=TintColor, sortPriority=7, defaultValue=<rgba 1 1 1 0>}
  n31 = MaterialFunctionCall!(
   DM=@n29.RGB
   Color1=@n30.RGB
   Color2=VectorParameter() {group=Material_1, parameterName=M1_TintColor2, sortPriority=5, defaultValue=<rgba 1 1 1 0>}.RGB
   Color3=VectorParameter() {group=Material_1, parameterName=M1_TintColor3, sortPriority=5, defaultValue=<rgba 1 1 1 0>}.RGB
  ) {materialFunction=/Game/Athena/WinterArt/Materials/MF_DMTexture}
  n32 = TextureObjectParameter() {group=Material_1, parameterName=Diffuse, texture=&T_IceCastle_Wall1_D}
  n33 = TextureSample(UVs=@n22, Tex=@n32) {texture=&T_IceCastle_Wall1_D}
  n34 = MaterialFunctionCall!(
   Floor1=@n30.RGB
   Floor2=VectorParameter() {group=Material_1, parameterName=Floor2, sortPriority=8, defaultValue=<rgba 1 1 1 0>}.RGB
   Floor3=VectorParameter() {group=Material_1, parameterName=Floor3, sortPriority=9, defaultValue=<rgba 1 1 1 0>}.RGB
   Basement=VectorParameter() {group=Material_1, parameterName=Basement, sortPriority=10, defaultValue=<rgba 1 1 1 0>}.RGB
  ) {materialFunction=/Game/Athena/WinterArt/Materials/MF_WS_GridUV}
  n35 = Multiply(
   A=@n33.RGB
   B=MaterialFunctionCall!(In=StaticSwitchParameter(True=@n34.StoryMaskOut, False=@n30.RGB) {group=Material_1, parameterName=UsePerFloorColor, sortPriority=4}) {materialFunction=/Game/Athena/WinterArt/Materials/MF_VPTintMask}.Result
  )
  n36 = TextureSampleParameter2D(UVs=@n25.UVs Layer 3) {
   desc=Use for trims and accents
   group=Material_3
   parameterName=Normals_Texture_3
   samplerSource=SSM_Wrap_WorldGroupSettings
   samplerType=SAMPLERTYPE_Normal
   sortPriority=3
   texture=&FlatNormal
  }
  n37 = TextureSampleParameter2D!(UVs=@n25.POM UVs 4 Layers, "DDX(UVs)"=DDX(Value=@n25.UVs Layer 3), "DDY(UVs)"=DDY(Value=@n25.UVs Layer 3), Apply View MipBias) {
   desc=Use for trims and accents
   group=Material_3
   mipValueMode=TMVM_Derivative
   parameterName=Normals_Texture_3
   samplerSource=SSM_Wrap_WorldGroupSettings
   samplerType=SAMPLERTYPE_Normal
   sortPriority=3
   texture=&FlatNormal
  }
  n38 = MaterialFunctionCall() {materialFunction=&MF_UVsSet}
  n39 = QualitySwitch(
   Default=@n38.UVs Layer 4
   High=StaticSwitchParameter(True=@n38.POM UVs 4 Layers, False=@n38.POM UVs 4 Layers) {group=Material_4, parameterName=L4_POM_True, sortPriority=11}
  )
  n40 = TextureSampleParameter2D!(UVs=@n39, Apply View MipBias) {
   desc=Use for doors
   group=Material_4
   parameterName=SpecularMasks_4
   samplerSource=SSM_Wrap_WorldGroupSettings
   samplerType=SAMPLERTYPE_Masks
   sortPriority=2
   texture=&T_Fortnite_Default_S
  }
  n41 = TextureObjectParameter() {
   group=Material_3
   parameterName=DM_3
   samplerType=SAMPLERTYPE_Masks
   sortPriority=1
   texture=&T_Fortnite_Default_S
  }
  n42 = TextureSample(Tex=@n41) {samplerType=SAMPLERTYPE_Masks}
  n43 = VectorParameter() {group=Material_3, parameterName=M3_TintColor, sortPriority=5, defaultValue=<rgba 1 1 1 0>}
  n44 = MaterialFunctionCall(
   DM=@n42.RGB
   Color1=@n43.RGB
   Color2=VectorParameter() {group=Material_3, parameterName=M3_TintColor2, sortPriority=5, defaultValue=<rgba 1 1 1 0>}.RGB
   Color3=VectorParameter() {group=Material_3, parameterName=M3_TintColor3, sortPriority=5, defaultValue=<rgba 1 1 1 0>}.RGB
  ) {materialFunction=&MF_DMTexture}
  n45 = TextureSampleParameter2D(UVs=@n26) {
   desc=Use for trims and accents
   group=Material_3
   parameterName=Diffuse_Texture_3
   samplerSource=SSM_Wrap_WorldGroupSettings
   sortPriority=1
   texture=/Game/Environments/Sets/Dojo/HomeBase/Textures/T_midgrey
  }
  n46 = Multiply(A=MaterialFunctionCall(In=@n43.RGB) {materialFunction=&MF_VPTintMask}.Result, B=@n45.RGB)
  n47 = TextureObjectParameter() {
   group=Material_4
   parameterName=DM_4
   samplerType=SAMPLERTYPE_Masks
   sortPriority=1
   texture=&T_Fortnite_Default_S
  }
  n48 = TextureSample(Tex=@n47) {samplerType=SAMPLERTYPE_Masks}
  n49 = VectorParameter() {group=Material_4, parameterName=M4_TintColor, sortPriority=5, defaultValue=<rgba 1 1 1 0>}
  n50 = MaterialFunctionCall(
   DM=@n48.RGB
   Color1=@n49.RGB
   Color2=VectorParameter() {group=Material_4, parameterName=M4_TintColor2, sortPriority=5, defaultValue=<rgba 1 1 1 0>}.RGB
   Color3=VectorParameter() {group=Material_4, parameterName=M4_TintColor3, sortPriority=5, defaultValue=<rgba 1 1 1 0>}.RGB
  ) {materialFunction=&MF_DMTexture}
  n51 = TextureSampleParameter2D(UVs=@n39) {
   desc=Use for doors
   group=Material_4
   parameterName=Diffuse_Texture_4
   samplerSource=SSM_Wrap_WorldGroupSettings
   sortPriority=1
   texture=&T_midgrey
  }
  n52 = Multiply(A=MaterialFunctionCall(In=@n49.RGB) {materialFunction=&MF_VPTintMask}.Result, B=@n51.RGB)
  n53 = TextureSampleParameter2D() {
   desc=Use for doors
   group=Material_4
   parameterName=Normals_Texture_4
   samplerSource=SSM_Wrap_WorldGroupSettings
   samplerType=SAMPLERTYPE_Normal
   sortPriority=3
   texture=&FlatNormal
  }
  n54 = TextureSampleParameter2D!(UVs=@n38.POM UVs 4 Layers, "DDX(UVs)"=DDX(Value=@n38.UVs Layer 4), "DDY(UVs)"=DDY(Value=@n38.UVs Layer 4), Apply View MipBias) {
   desc=Use for doors
   group=Material_4
   mipValueMode=TMVM_Derivative
   parameterName=Normals_Texture_4
   samplerSource=SSM_Wrap_WorldGroupSettings
   samplerType=SAMPLERTYPE_Normal
   sortPriority=3
   texture=&FlatNormal
  }
  n55 = MaterialFunctionCall() {materialFunction=&MF_UVsSet}
  n56 = QualitySwitch(
   Default=@n55.UVs Layer 2
   High=StaticSwitchParameter(True=@n55.POM UVs 4 Layers, False=@n55.POM UVs 4 Layers) {group=Material_2, parameterName=L2_POM_True, sortPriority=11}
  )
  n57 = TextureObjectParameter() {
   group=Material_2
   parameterName=DM_2
   samplerType=SAMPLERTYPE_Masks
   sortPriority=1
   texture=&T_Fortnite_Default_S
  }
  n58 = TextureSample(UVs=@n56, Tex=@n57) {samplerType=SAMPLERTYPE_Masks}
  n59 = VectorParameter() {group=Material_2, parameterName=M2_TintColor, sortPriority=5, defaultValue=<rgba 1 1 1 0>}
  n60 = MaterialFunctionCall(
   DM=@n58.RGB
   Color1=@n59.RGB
   Color2=VectorParameter() {group=Material_2, parameterName=M2_TintColor2, sortPriority=5, defaultValue=<rgba 1 1 1 0>}.RGB
   Color3=VectorParameter() {group=Material_2, parameterName=M2_TintColor3, sortPriority=5, defaultValue=<rgba 1 1 1 0>}.RGB
  ) {materialFunction=&MF_DMTexture}
  n61 = TextureSampleParameter2D!(UVs=@n56, Apply View MipBias) {
   desc=Use for interior walls and ceilings
   group=Material_2
   parameterName=Diffuse_Texture_2
   samplerSource=SSM_Wrap_WorldGroupSettings
   sortPriority=1
   texture=&T_midgrey
  }
  n62 = Multiply(A=MaterialFunctionCall(In=@n59.RGB) {materialFunction=&MF_VPTintMask}.Result, B=@n61.RGB)
  n63 = TextureSampleParameter2D(UVs=@n56) {
   desc=Use for interior walls and ceilings
   group=Material_2
   parameterName=SpecularMasks_2
   samplerSource=SSM_Wrap_WorldGroupSettings
   samplerType=SAMPLERTYPE_Masks
   sortPriority=2
   texture=&T_Fortnite_Default_S
  }
  n64 = TextureSampleParameter2D(UVs=@n55.UVs Layer 2) {
   desc=Use for interior walls and ceilings
   group=Material_2
   parameterName=Normals_Texture_2
   samplerSource=SSM_Wrap_WorldGroupSettings
   samplerType=SAMPLERTYPE_Normal
   sortPriority=3
   texture=&FlatNormal
  }
  n65 = TextureSampleParameter2D!(UVs=@n55.POM UVs 4 Layers, "DDX(UVs)"=DDX(Value=@n55.UVs Layer 2), "DDY(UVs)"=DDY(Value=@n55.UVs Layer 2), Apply View MipBias) {
   desc=Use for interior walls and ceilings
   group=Material_2
   mipValueMode=TMVM_Derivative
   parameterName=Normals_Texture_2
   samplerSource=SSM_Wrap_WorldGroupSettings
   samplerType=SAMPLERTYPE_Normal
   sortPriority=3
   texture=&FlatNormal
  }
  n66 = MaterialFunctionCall!(
   D=StaticSwitchParameter(True=AppendVector(A=@n31.Result, B=@n29.A), False=AppendVector(A=@n35, B=@n33.A)) {desc=Creative, group=Material_1, parameterName=LM_DMTextureTrue, sortPriority=11}
   S=AppendVector(A=@n24.RGB, B=@n24.RGB)
   N=QualitySwitch(Default=@n21.RGB, High=StaticSwitchParameter(True=@n20.RGB, False=@n21.RGB) {group=Material_1, parameterName=POM_True, sortPriority=11})
   D2=StaticSwitchParameter(True=AppendVector(A=@n60.Result, B=@n58.A), False=AppendVector(A=@n62, B=@n61.A)) {desc=Creative, group=Material_2, parameterName=L2_DMTextureTrue, sortPriority=11}
   S2=AppendVector(A=@n63.RGB, B=@n63.RGB)
   N2=QualitySwitch(Default=@n64.RGB, High=StaticSwitchParameter(True=@n65.RGB, False=@n64.RGB) {group=Material_2, parameterName=L2_POM_True, sortPriority=11})
   D3=StaticSwitchParameter(True=AppendVector(A=@n44.Result, B=@n42.A), False=AppendVector(A=@n46, B=@n45.A)) {desc=Creative, group=Material_3, parameterName=L3_DMTextureTrue, sortPriority=11}
   S3=AppendVector(A=@n27.RGB, B=@n27.RGB)
   N3=QualitySwitch(Default=@n36.RGB, High=StaticSwitchParameter(True=@n37.RGB, False=@n36.RGB) {group=Material_3, parameterName=L3_POM_True, sortPriority=11})
   D4=StaticSwitchParameter(True=AppendVector(A=@n50.Result, B=@n48.A), False=AppendVector(A=@n52, B=@n51.A)) {desc=Creative, group=Material_3, parameterName=L3_DMTextureTrue, sortPriority=11}
   S4=AppendVector(A=@n40.RGB, B=@n40.RGB)
   N4=QualitySwitch(Default=@n53.RGB, High=StaticSwitchParameter(True=@n54.RGB, False=@n53.RGB) {group=Material_4, parameterName=L4_POM_True, sortPriority=11})
   E1=VectorParameter() {group=Material_1, parameterName=Emissive, sortPriority=7}.RGB
   E2=VectorParameter() {group=Material_2, parameterName=Emissive2, sortPriority=7}.RGB
   E3=VectorParameter() {group=Material_3, parameterName=Emissive3, sortPriority=7}.RGB
   E4=VectorParameter() {group=Material_4, parameterName=Emissive4, sortPriority=7}.RGB
  ) {materialFunction=/FigmentEnvironment/LM_8_1_1/Athena/WinterArt/Materials/MF_FN_TextureSetLerp_4Layers}
  n67 = Constant() {r=1}
  n68 = ScalarParameter() {defaultValue=0.7, parameterName=DetailDiffuseDarkness}
  n69 = TextureSample(UVs=Multiply(A=TextureCoordinate(), B=ScalarParameter() {defaultValue=1.5, parameterName=DetailRoughTile})) {samplerType=SAMPLERTYPE_Masks, texture=/Game/Athena/Prototype/Blueprints/Cube/Materials/T_CubeNoise}
  n70 = ScalarParameter() {group=Global, parameterName=Spec Min Fresnel}
  n71 = Fresnel(ExponentIn=ScalarParameter() {defaultValue=5, group=Global, parameterName=SpecFresnelExp}, Normal=VertexNormalWS()) {exponent=3}
  n72 = TextureSample(UVs=Multiply(A=TextureCoordinate(), B=ScalarParameter() {defaultValue=8, parameterName=DetailNormalTile})) {samplerType=SAMPLERTYPE_Normal, texture=/Figment_S08/Athena/Environments/Nexus/Textures/NexFloor_2_N}
  n73 = Constant3Vector() {constant=<rgba 1 1 0 0>}
  n74 = Add(A=@n66.Normals, B=Multiply(A=Multiply(A=@n72.RGB, B=@n73), B=ScalarParameter() {defaultValue=0.6, parameterName=DetailNormalAmount}))
  n75 = LinearInterpolate(
   A=ScalarParameter() {defaultValue=0.6, parameterName=DetailRough-Low}
   B=ScalarParameter() {defaultValue=0.85, parameterName=DetailRough-High}
   Alpha=@n69.R
  )
  n76 = StaticComponentMaskParameter(None=TextureSampleParameter2D!(UVs, Apply View MipBias) {group=Emissive, parameterName=Emissive Mask Texture, texture=&T_BlacksRGB}.RGB) {defaultG=true, group=Emissive, parameterName=Emissive Texture Mask Channel}
  n77 = Multiply(A=@n76, B=Clamp(None=@n5, Min=ScalarParameter() {group=Emissive, parameterName=Emissive Flicker Min Clamp}) {maxDefault=3})
  n78 = ScalarParameter() {group=Emissive, parameterName=Emissive Flicker Lerp}
  n79 = SetMaterialAttributes!(
   MaterialAttributes
   Base Color=Multiply(A=Add(A=@n66.Diffuse, B=@n15), B=LinearInterpolate(A=@n67, B=@n68, Alpha=@n69.R))
   Specular=LinearInterpolate(A=Multiply(A=@n70, B=@n66.S), B=@n66.S, Alpha=OneMinus(None=@n71))
   Metallic=@n66.M
   Roughness=LinearInterpolate(A=Multiply(A=@n66.R, B=@n75), B=@n66.R, Alpha=@n14.Result)
   Emissive Color=Multiply(A=LinearInterpolate(A=@n76, B=@n77, Alpha=@n78), B=@n66.Emissive)
   Normal=Normalize(VectorInput=LinearInterpolate(A=@n74, B=@n66.Normals, Alpha=@n14.Result))
  ) {attributeSetTypes [
    #1
    #2
    #3
    #4
    #5
    #6
   ]}
  Reroute(None=@n79) {desc=ES3_1}
  Reroute(None=@n66.Diffuse)
  n80 = MaterialFunctionCall!(
   In=MaterialFunctionCall!(transitionFactor, MaterialIn=@n79) {materialFunction=/Game/Athena/WinterArt/Materials/MF_WSGradient}.Color*Gradient
   BGMap=@n66.Height
  ) {materialFunction=/FigmentEnvironment/LM_8_1_1/Athena/WinterArt/Materials/MF_VPTexture}
  n81 = MaterialFunctionCall!(In=@n80.Result, BG_HeightMap=@n66.Height, UpVectorTexture=@n80.Result) {materialFunction=/FigmentEnvironment/LM_8_1_1/Athena/WinterArt/Materials/MF_UpVectorTexture}
  n82 = SetMaterialAttributes(Base Color=GetMaterialAttributes(None=@n81.Result) {attributeGetTypes [#1]}.Base Color) {attributeSetTypes [
    #1
    #2
    #3
    #4
    #5
    #6
   ]}
  n83 = MaterialFunctionCall!(VibrationAmount, WobbleScale, WobbleDuration) {materialFunction=&/MF_LakeStructure_HitWPO}
  n84 = MaterialFunctionCall!(
   Material=MaterialFunctionCall!(
    Material=FeatureLevelSwitch(Default=@n81.Result, ES2=@n82, ES3_1=@n81.Result, SM4_REMOVED=@n81.Result, SM5=@n81.Result, SM6=@n81.Result)
    Mask
    WorldPositionOffset=@n83.Result
   ) {materialFunction=/Engine/Functions/MaterialLayerFunctions/MatLayerBlend_OverrideWorldPositionOffset}.Blended Material
   NormalFlatness=@n83.Mask
   Normal=Constant3Vector() {constant=<rgba 0 0 2 0>}
  ) {materialFunction=/Engine/Functions/MaterialLayerFunctions/MatLayerBlend_NormalFlatten}
  n85 = Constant() {r=0.02}
  SetMaterialAttributes!(
   MaterialAttributes=MaterialFunctionCall!(
    Material=MaterialFunctionCall!(Material=@n84.Blended Material, NewRoughness=@n85, Mask=@n83.Mask) {materialFunction=/Engine/Functions/MaterialLayerFunctions/MatLayerBlend_RoughnessOverride}.Blended Material
    NewBaseColor=VectorParameter() {parameterName=FXBaseColor, defaultValue=<rgba 0.1347 0.1324 0.16 1>}.RGB
    Mask=@n83.Mask
   ) {materialFunction=/Engine/Functions/MaterialLayerFunctions/MatLayerBlend_OverrideBaseColor}.Blended Material
   World Position Offset=StaticSwitchParameter(True=MaterialFunctionCall() {materialFunction=/Game/Environments/Sets/DigSite/Materials/MF_Sneezy_DanceClub}.Result, False=Constant()) {parameterName=DanceClub}
  ) {attributeSetTypes [#7]}
  n86 = TextureObjectParameter() {group=Material_2, parameterName=Diffuse_Texture_2, sortPriority=1, texture=&T_midgrey}
  StaticSwitchParameter(True=@n57, False=@n86) {desc=Creative, group=Material_2, parameterName=L2_DMTextureTrue, sortPriority=11}
  n87 = TextureObjectParameter() {group=Material_4, parameterName=Diffuse_Texture_4, sortPriority=1, texture=&T_midgrey}
  StaticSwitchParameter(True=@n47, False=@n87) {desc=Creative, group=Material_3, parameterName=L3_DMTextureTrue, sortPriority=11}
  StaticSwitchParameter(True=@n28, False=@n32) {desc=Creative, group=Material_1, parameterName=L1_DMTextureTrue, sortPriority=11}
  n88 = TextureObjectParameter() {group=Material_3, parameterName=Diffuse_Texture_3, sortPriority=1, texture=&T_midgrey}
  StaticSwitchParameter(True=@n41, False=@n88) {desc=Creative, group=Material_3, parameterName=L3_DMTextureTrue, sortPriority=11}
 ]
}
